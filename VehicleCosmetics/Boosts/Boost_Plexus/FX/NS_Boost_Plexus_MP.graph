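# Niagara system for the Plexus vehicle boost cosmetic, streaming animated triangle sprites backward from the boost position.
# ThrusterTris spawns fading, curl-noise-scattered sprites; WireframeTris adds larger, straight-moving sprites, both using multiple SubUV-animated renderers.
# Both tint through painted-variant overrides, spawn around System.FinalPosition and inherit source movement with falloff over lifetime.
# User.bIsFrontendPreview adds drag to ThrusterTris and changes WireframeTris velocity for frontend previews.
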NiagaraSystem /VehicleCosmetics/Boosts/Boost_Plexus/FX/NS_Boost_Plexus_MP
dependencies {
 usedDataInterfaces [
  &NiagaraDataInterfaceCurve
  &NiagaraDataInterfaceSpriteRendererInfo
  &NiagaraDataInterfaceSkeletalMesh
 ]
 usedDynamicInputs [&RandomRangeFloat]
 usedModules [
  &self:SystemSpawnScript.SystemScriptSource.SystemScriptGraph.NiagaraNodeAssignment_0.SetVariables_#1
  &SystemState
 ]
}
emitters {
 ThrusterTris {
  emitterUpdateScript [
   &EmitterState()
   &SpawnRate(SpawnRate=30, Use Spawn Probability=false)
  ]
  particleSpawnScript [
   &VC_PaintedVariantsOverride(
    Black=<rgba 0.005 0.005 0.005 1>
    BurntSienna=<rgba 0.1042 0.0352 0.0128 1>
    Cobalt=<rgba 0 0.0852 1 1>
    Crimson=<rgba 0.4635 0 0.0126 1>
    DefaultColor=<rgba 0.1017 0.1961 0.3906 1>
    ForestGreen=<rgba 0 0.4271 0.0216 1>
    Lime=<rgba 0.2562 0.375 0 1>
    Orange=<rgba 0.2917 0.2236 0 1>
    Pink=<rgba 0.4745 0.1695 0.651 1>
    Purple=<rgba 0.218 0 0.5312 1>
    R1 Black:LinearColor=<rgba 0 0 0 1>
    R1 BurntSienna:LinearColor=<rgba 0.3438 0.1061 0.0424 1>
    R1 Cobalt:LinearColor=<rgba 0.224 0.2383 1 1>
    R1 Crimson:LinearColor=<rgba 1 0.0781 0.1031 1>
    R1 DefaultColor=<rgba 0.4167 0.5587 1 1>
    R1 ForestGreen:LinearColor=<rgba 0.1823 1 0.2236 1>
    R1 Grey:LinearColor=<rgba 0.5208 0.5208 0.5208 1>
    R1 Lime:LinearColor=<rgba 0.5 1 0 1>
    R1 Orange:LinearColor=<rgba 0.8594 0.4583 0 1>
    R1 Pink:LinearColor=<rgba 0.7766 0.3906 1 1>
    R1 Purple:LinearColor=<rgba 0.4442 0.0573 1 1>
    R1 SkyBlue:LinearColor=<rgba 0 0.3167 1 1>
    Saffron=<rgba 0.5521 0.4233 0 1>
    SkyBlue=<rgba 0 0.1699 0.5365 1>
    TitaniumWhite=<rgba 0.3594 0.3594 0.3594 1>
    Use R1 Painted Overrides=true
   )
   &InitializeParticle(
    Color Mode=::NewEnumerator0
    Lifetime=0.5
    Position Offset=&AddVectorToPosition(Position=Linked() {linkedVariable {name=System.FinalPosition, type=Vector3f}}, Vector=<0 0 0>)
    Position Offset Coordinate Space=::Local
    Sprite Rotation Mode=::NewEnumerator2
    Sprite Size Mode=::NewEnumerator3
    Uniform Sprite Size=80
    UsePositionOffset=true
    Write Material Random=true
   )
   &ShapeLocation(MaskSpawnGroup=false, Override Seed=false, Sphere Radius=40, UseOwnerScale=true)
   &AddVelocity(Velocity=&RandomRangeVector(Minimum=<-1000 -400 -90>, Maximum=<-800 400 300>))
  ]
  particleUpdateScript [
   &ParticleState()
   &InheritSourceMovement(
    Applied Position Scale=&VectorFromFloat(Value=&FloatFromCurve(FloatCurve=DataInterface() {propertyValues {Curve {keys [
          {interpMode=RCIM_Cubic, tangentMode=RCTM_Break, time=0, value=1}
          {arriveTangent=-2.9025, interpMode=RCIM_Cubic, tangentMode=RCTM_Break, time=1, value=0}
         ]}}}))
    ApplyForce=false
    ApplyPosition=true
    ApplyPositionFalloff=false
   )
   &CurlNoiseForce(Bias Noise=false, Noise Strength=700, Pan Noise=true)
   &Drag(
    Drag=&MakeCustomFloatFromBool(
     Bool=$User.bIsFrontendPreview
     True Float=&FloatFromCurve(FloatCurve=DataInterface() {propertyValues {Curve {keys [
          {arriveTangent=1.6846, interpMode=RCIM_Cubic, leaveTangent=1.6846, tangentMode=RCTM_User, time=0.3, value=0}
          {arriveTangent=0.0224, interpMode=RCIM_Cubic, leaveTangent=0.0224, tangentMode=RCTM_User, time=1, value=0.675}
         ]}}})
     False Float=0
    )
    Rotational Drag:NiagaraFloat=&MakeCustomFloatFromBool(Bool=$User.bIsFrontendPreview, True Float=1, False Float=0)
    Use Linear Drag=true
    Use Rotational Drag=true
   )
   &ScaleSpriteSize(Uniform Curve Sprite Scale=DataInterface() {propertyValues {Curve {keys [<at 0 1 RCIM_Cubic>,<at 1.8 0 RCIM_Cubic>]}}}) {module=ScaleSpriteSize001}
   &SubUVAnimation(
    Random Start Frame=true
    "Sprite Renderer "=DataInterface() {propertyValues {SpriteRenderer=&self:ThrusterTris_0.NiagaraSpriteRendererProperties_1}}
    UseEndFrame=false
    UseStartFrame=false
   ) {module=SubUVAnimation001}
   &SubUVAnimation(
    Random Start Frame=true
    "Sprite Renderer "=DataInterface() {propertyValues {SpriteRenderer=&self:ThrusterTris_0.NiagaraSpriteRendererProperties_0}}
    UseEndFrame=false
    UseStartFrame=false
   )
   &SolveForcesAndVelocity(Clamp Velocity=false, Limit Acceleration=false)
   &ScaleColor(
    Scale Alpha=&FloatFromCurve(FloatCurve=DataInterface() {propertyValues {Curve {keys [<at 0 1>,<at 1 0>]}}})
    ScaleA=true
    ScaleRGB=true
   )
  ]
  rendererClasses [&NiagaraSpriteRendererProperties]
  renderers [
   {rendererClass=&NiagaraSpriteRendererProperties}
   {rendererClass=&NiagaraSpriteRendererProperties, rendererIndex=1}
  ]
 }
 WireframeTris {
  emitterUpdateScript [
   &EmitterState()
   &SpawnRate(SpawnRate=25, Use Spawn Probability=false)
  ]
  particleSpawnScript [
   &VC_PaintedVariantsOverride(
    Black=<rgba 0.02 0.02 0.02 1>
    BurntSienna=<rgba 0.2708 0.1153 0.0334 1>
    Cobalt=<rgba 0.0752 0.1521 0.8021 1>
    DefaultColor=<rgba 0.1447 0.2705 0.6615 1>
    ForestGreen=<rgba 0 0.4635 0.0234 1>
    Lime=<rgba 0.3629 0.5885 0 1>
    Orange=<rgba 0.3385 0.2641 0.0194 0.913>
    Pink=<rgba 0.4547 0.1335 0.6406 0.6348>
    Purple=<rgba 0.1945 0 0.474 0.8435>
    R1 Black:LinearColor=<rgba 0 0 0 1>
    R1 BurntSienna:LinearColor=<rgba 0 0 0 1>
    R1 Cobalt:LinearColor=<rgba 0 0 0 1>
    R1 Crimson:LinearColor=<rgba 0 0 0 1>
    R1 DefaultColor=<rgba 0 0 0 1>
    R1 ForestGreen:LinearColor=<rgba 0 0 0 1>
    R1 Grey:LinearColor=<rgba 0 0 0 1>
    R1 Lime:LinearColor=<rgba 0 0 0 1>
    R1 Orange:LinearColor=<rgba 0 0 0 1>
    R1 Pink:LinearColor=<rgba 0 0 0 1>
    R1 Purple:LinearColor=<rgba 0 0 0 1>
    R1 Saffron:LinearColor=<rgba 0 0 0 1>
    R1 SkyBlue:LinearColor=<rgba 0 0 0 1>
    R1 TitaniumWhite:LinearColor=<rgba 0 0 0 1>
    R2 Black:LinearColor=<rgba 0.02 0.02 0.02 1>
    R2 BurntSienna:LinearColor=<rgba 0.2708 0.1153 0.0334 1>
    R2 Cobalt:LinearColor=<rgba 0.375 0.5469 2 1>
    R2 DefaultColor:LinearColor=<rgba 0.4219 0.5626 1 1>
    R2 Grey:LinearColor=<rgba 0.4323 0.4323 0.4323 1>
    R2 Lime:LinearColor=<rgba 0.6673 1 0 1>
    R2 Orange:LinearColor=<rgba 0.6724 0.5155 0 1>
    R2 Saffron:LinearColor=<rgba 0.9 0.69 0 1>
    R2 SkyBlue:LinearColor=<rgba 0 0.1699 0.5365 1>
    Saffron=<rgba 0.4792 0.3674 0 0.8609>
    SkyBlue=<rgba 0 0.138 0.4375 1>
    TitaniumWhite=<rgba 0.3958 0.3958 0.3958 1>
    Use R1 Painted Overrides=true
    Use R2 Painted Overrides=true
   )
   &InitializeParticle(
    Color Mode=::NewEnumerator0
    Lifetime=0.5
    Position Offset=&AddVectorToPosition(Position=Linked() {linkedVariable {name=System.FinalPosition, type=Vector3f}}, Vector=<0 0 0>)
    Position Offset Coordinate Space=::Local
    Sprite Rotation Mode=::NewEnumerator2
    Sprite Size Mode=::NewEnumerator1
    Uniform Sprite Size Max=250
    Uniform Sprite Size Min=150
    UsePositionOffset=true
    Write Material Random=true
   )
   &ShapeLocation(MaskSpawnGroup=false, Override Seed=false, Sphere Radius=40, UseOwnerScale=true)
   &AddVelocity(Velocity=&MakeCustomVectorFromBool(Bool=$User.bIsFrontendPreview, True Vector=<-1200 0 10>, False Vector=<-1000 0 0>))
  ]
  particleUpdateScript [
   &ParticleState()
   &InheritSourceMovement(
    Applied Position Scale=&VectorFromFloat(Value=&FloatFromCurve(FloatCurve=DataInterface() {propertyValues {Curve {keys [
          {interpMode=RCIM_Cubic, tangentMode=RCTM_Break, time=0, value=1}
          {arriveTangent=-2.9025, interpMode=RCIM_Cubic, tangentMode=RCTM_Break, time=1, value=0}
         ]}}}))
    ApplyForce=false
    ApplyPosition=true
    ApplyPositionFalloff=false
   )
   &ScaleSpriteSize(
    Uniform Curve Scale=0.6
    Uniform Curve Sprite Scale=DataInterface() {propertyValues {Curve {keys [
        <at 0 0.8 RCIM_Cubic>
        {arriveTangent=-0.0571, interpMode=RCIM_Cubic, leaveTangent=-0.0571, tangentMode=RCTM_User, time=0.8, value=1.5}
        <at 1 0 RCIM_Cubic>
       ]}}}
   ) {module=ScaleSpriteSize001}
   &SubUVAnimation(
    Random Start Frame=true
    "Sprite Renderer "=DataInterface() {propertyValues {SpriteRenderer=&self:WireframeTris_0.NiagaraSpriteRendererProperties_2}}
    SubUV Lookup Index:NiagaraFloat=&Multiply_Float(A=$Particles.NormalizedAge, B=0.5)
    UseEndFrame=false
    UseStartFrame=false
   ) {module=SubUVAnimation001}
   &SubUVAnimation(
    Random Start Frame=true
    "Sprite Renderer "=DataInterface() {propertyValues {SpriteRenderer=&self:WireframeTris_0.NiagaraSpriteRendererProperties_1}}
    SubUV Lookup Index:NiagaraFloat=&Multiply_Float(A=$Particles.NormalizedAge, B=0.5)
    UseEndFrame=false
    UseStartFrame=false
   ) {module=SubUVAnimation002}
   &SubUVAnimation(
    Random Start Frame=true
    "Sprite Renderer "=DataInterface() {propertyValues {SpriteRenderer=&self:WireframeTris_0.NiagaraSpriteRendererProperties_0}}
    SubUV Lookup Index:NiagaraFloat=&Multiply_Float(A=$Particles.NormalizedAge, B=0.5)
    UseEndFrame=false
    UseStartFrame=false
   )
   &SolveForcesAndVelocity(Clamp Velocity=false, Limit Acceleration=false)
  ]
  rendererClasses [&NiagaraSpriteRendererProperties]
  renderers [
   {rendererClass=&NiagaraSpriteRendererProperties}
   {rendererClass=&NiagaraSpriteRendererProperties, rendererIndex=1}
   {rendererClass=&NiagaraSpriteRendererProperties, rendererIndex=2}
  ]
 }
}
summary {
 systemName=NS_Boost_Plexus_MP
 emitters [
  {emitterName=ThrusterTris, rendererClasses [&NiagaraSpriteRendererProperties]}
  {emitterName=WireframeTris, rendererClasses [&NiagaraSpriteRendererProperties]}
 ]
 userVariables [
  User.bIsFrontendPreview {type=NiagaraBool}
  User.bIsPainted {type=NiagaraBool}
  User.ColorVariantID {type=NiagaraInt32}
  User.DTCarOffset {type=Vector3f, description="", defaultValue {struct=Vector3f, value=<0 0 0>}}
  User.GameModeID {type=NiagaraInt32}
  User.PaintedColor {type=LinearColor, description="", defaultValue {struct=LinearColor, value=<rgba 1 1 1 1>}}
  User.PaintedScalar {type=NiagaraFloat, defaultValue=1, description=""}
  User.PortCount {type=NiagaraInt32, defaultValue=1}
  User.SingleBoostPos {type=NiagaraPosition, description="", defaultValue {}}
 ]
}
