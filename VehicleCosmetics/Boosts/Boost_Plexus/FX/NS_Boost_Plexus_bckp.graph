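# Niagara vehicle boost trail for the Plexus cosmetic, apparently a backup copy by its name.
# Six emitters layer two ribbons, short-lived mesh/sprite particles, light-emitting sprites and two SubUV puff layers.
# Filtered socket counts distribute ribbons and particles across skeletal-mesh ports, streaming backward along -X and briefly inheriting vehicle movement.
# Curves drive dynamic material parameters and mostly blue-to-warm color shifts; painted-variant overrides are disabled in every emitter.
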NiagaraSystem /VehicleCosmetics/Boosts/Boost_Plexus/FX/NS_Boost_Plexus_bckp
dependencies {
 usedDataInterfaces [
  &NiagaraDataInterfaceCurve
  &NiagaraDataInterfaceColorCurve
  &NiagaraDataInterfaceMeshRendererInfo
  &NiagaraDataInterfaceVector2DCurve
  &NiagaraDataInterfaceSpriteRendererInfo
  &NiagaraDataInterfaceSkeletalMesh
 ]
 usedDynamicInputs []
 usedModules [
  &self:SystemSpawnScript.SystemScriptSource.SystemScriptGraph.NiagaraNodeAssignment_0.SetVariables_#1
  &SystemState
 ]
}
emitters {
 NE_VC_Boost_Ribbon_Parent {
  emitterUpdateScript [
   &EmitterState()
   &VC_FilteredSocketCount()
   &VC_SpawnRibbon_v1(Number of Ribbons=$Emitter.PortCount, SpawnRate=50)
  ]
  particleSpawnScript [
   &VC_PaintedVariantsOverride() {enabled=false}
   &VC_InitializeRibbon_v1(Color=<rgba 2 2 2 1>, Enable Skeletal Position=true, Facing Axis:ENiagaraOrientationAxis=::ZAxis, Lifetime=0.4, Ribbon Width=65)
   &AddVelocity(Velocity=<-1000 0 0>) {module=AddVelocity001}
  ]
  particleUpdateScript [
   &ParticleState()
   &InheritSourceMovement(
    Applied Position Scale=&Multiply_VectorByFloat(
     Vector=<1 1 1>
     Float=&FloatFromCurve(FloatCurve=DataInterface() {propertyValues {Curve {
         defaultValue=0
         keys [
          {tangentMode=RCTM_User, time=0, value=1}
          <at 0.03 0.489>
          <at 0.06 0.1856>
          <at 0.1 0.0532>
          <at 0.15 0>
          {tangentMode=RCTM_User, time=1, value=0}
         ]
        }}})
    )
    ApplyForce=false
    ApplyPosition=true
    ApplyPositionFalloff=false
   )
   &ScaleColor(
    Scale Alpha=&FloatFromCurve(
     FloatCurve=DataInterface() {propertyValues {Curve {
        defaultValue=1
        keys [
         {arriveTangent=-2.385, interpMode=RCIM_Cubic, leaveTangent=-2.385, tangentMode=RCTM_User, time=0, value=1}
         {arriveTangent=0.03, interpMode=RCIM_Cubic, leaveTangent=0.03, tangentMode=RCTM_User, time=1, value=0}
        ]
       }}}
     Scale Curve=0.8
    )
    Scale RGB=&MakeVectorFromLinearColorRGB(Color=&ColorFromCurve(ColorCurve=DataInterface() {propertyValues {
        AlphaCurve {keys [<at 0 1 RCIM_Cubic>,<at 1 0 RCIM_Cubic>]}
        BlueCurve {keys [<at 0.0023 80>,<at 0.1014 5>,<at 0.2407 0>,<at 0.5302 0>,<at 0.9853 0>]}
        GreenCurve {keys [<at 0.0023 3>,<at 0.1014 0.9217>,<at 0.2407 0.6529>,<at 0.5302 0.1179>,<at 0.9853 0.0293>]}
        RedCurve {keys [<at 0.0023 0.5>,<at 0.1014 0.3296>,<at 0.2407 5>,<at 0.5302 2.9427>,<at 0.9853 2.1304>]}
       }}))
    ScaleA=true
    ScaleRGB=true
   )
   &CurlNoiseForce(Bias Noise=false, Noise Frequency=2000, Noise Strength=3000, Pan Noise=true, Pan Noise Field=<1 0 0>) {enabled=false}
   &SolveForcesAndVelocity(Clamp Velocity=false, Limit Acceleration=false)
   &VC_UpdateRibbon_v1(Enable Skeletal Position=true)
   &ScaleRibbonWidth(Ribbon Width Scale=&FloatFromCurve(FloatCurve=DataInterface() {propertyValues {Curve {keys [<at 0 1>,<at 1 0>]}}})) {enabled=false}
   &ScaleRibbonWidth(Ribbon Width Scale=&FloatFromCurve(FloatCurve=DataInterface() {propertyValues {Curve {keys [<at 0 1.5>,<at 1 1.5>]}}})) {enabled=false, module=ScaleRibbonWidth001}
  ]
  rendererClasses [&NiagaraRibbonRendererProperties]
  renderers [{rendererClass=&NiagaraRibbonRendererProperties}]
 }
 NE_VC_Boost_Ribbon_Parent002 {
  emitterUpdateScript [
   &EmitterState()
   &VC_FilteredSocketCount()
   &VC_SpawnRibbon_v1(Number of Ribbons=$Emitter.PortCount, SpawnRate=50)
  ]
  particleSpawnScript [
   &VC_PaintedVariantsOverride() {enabled=false}
   &VC_InitializeRibbon_v1(Color=<rgba 0.2615 0.1745 1 0.8>, Enable Skeletal Position=true, Facing Axis:ENiagaraOrientationAxis=::ZAxis, Lifetime=0.4, Ribbon Width=75)
   &AddVelocity(Velocity=<-1000 0 0>) {module=AddVelocity001}
  ]
  particleUpdateScript [
   &ParticleState()
   &InheritSourceMovement(
    Applied Position Scale=&Multiply_VectorByFloat(
     Vector=<1 1 1>
     Float=&FloatFromCurve(FloatCurve=DataInterface() {propertyValues {Curve {
         defaultValue=0
         keys [
          {tangentMode=RCTM_User, time=0, value=1}
          <at 0.03 0.489>
          <at 0.06 0.1856>
          <at 0.1 0.0532>
          <at 0.15 0>
          {tangentMode=RCTM_User, time=1, value=0}
         ]
        }}})
    )
    ApplyForce=false
    ApplyPosition=true
    ApplyPositionFalloff=false
   )
   &CurlNoiseForce(Bias Noise=false, Noise Frequency=200, Noise Strength=800, Pan Noise=true, Pan Noise Field=<0 0 200>)
   &ScaleColor(
    Scale Alpha=&FloatFromCurve(FloatCurve=DataInterface() {propertyValues {Curve {
        defaultValue=1
        keys [
         {arriveTangent=-0.0472, interpMode=RCIM_Cubic, leaveTangent=-0.0472, tangentMode=RCTM_User, time=0, value=0}
         {arriveTangent=-2.9208, interpMode=RCIM_Cubic, leaveTangent=-2.9208, tangentMode=RCTM_User, time=0.04, value=0.8}
         {arriveTangent=-0.0641, interpMode=RCIM_Cubic, leaveTangent=-0.0641, tangentMode=RCTM_User, time=0.4, value=0}
        ]
       }}})
    Scale RGB=&MakeVectorFromLinearColorRGB(Color=&ColorFromCurve(ColorCurve=DataInterface() {propertyValues {
        AlphaCurve {keys [<at 0 1 RCIM_Cubic>,<at 1 0 RCIM_Cubic>]}
        BlueCurve {keys [<at 0.1208 20>,<at 0.2613 0>]}
        GreenCurve {keys [<at 0.1208 2.3412>,<at 0.2613 3.8246>]}
        RedCurve {keys [<at 0.1208 0.625>,<at 0.2613 20>]}
       }}))
    ScaleA=true
    ScaleRGB=true
   )
   &SolveForcesAndVelocity(Clamp Velocity=false, Limit Acceleration=false)
   &VC_UpdateRibbon_v1(Enable Skeletal Position=true)
   &ScaleRibbonWidth(Ribbon Width Scale=&FloatFromCurve(FloatCurve=DataInterface() {propertyValues {Curve {
        defaultValue=1
        keys [
         {arriveTangent=0.5578, interpMode=RCIM_Cubic, leaveTangent=0.5578, tangentMode=RCTM_User, time=0, value=1}
         {arriveTangent=3.5117, interpMode=RCIM_Cubic, leaveTangent=3.5117, tangentMode=RCTM_User, time=0.4, value=1.5}
        ]
       }}})) {enabled=false}
  ]
  rendererClasses [&NiagaraRibbonRendererProperties]
  renderers [{rendererClass=&NiagaraRibbonRendererProperties}]
 }
 NE_VC_Boost_Sprite_Parent {
  emitterUpdateScript [
   &EmitterState()
   &VC_FilteredSocketCount()
   &SpawnRate(SpawnRate=200, Use Spawn Probability=false)
  ]
  particleSpawnScript [
   &VC_PaintedVariantsOverride() {enabled=false}
   &InitializeParticle(
    AdjustAlpha=false
    AdjustHue=true
    AdjustSaturation=false
    AdjustValue=false
    Color=<rgba 2 2 2 1>
    Color Mode=::NewEnumerator3
    Hue Shift Range:Vector2f=<-0.5 0>
    Lifetime=&RandomRangeFloat(Minimum=0.08, Maximum=0.11)
    Mesh Renderer Array Visibility Mode=::NewEnumerator3
    Mesh Renderer Info=DataInterface() {propertyValues {MeshRenderer=&self:NE_VC_Boost_Sprite_Parent.NiagaraMeshRendererProperties_0}}
    Mesh Scale Mode=::NewEnumerator1
    Mesh Uniform Scale Max=0.1
    Mesh Uniform Scale Min=0.03
    Sprite Rotation Mode=::NewEnumerator2
    Sprite Size Mode=::NewEnumerator3
    UsePositionOffset=false
    Write Material Random=true
   )
   &InitialMeshOrientation(Use Rotation Vector=false)
   &PartitionParticles(Number Of Partitions=$Emitter.PortCount)
   &SkeletalMeshLocation(
    Bone / Socket Index=$Particles.PartitionParticles.Partition
    Bone / Socket Sampling Mode=::NewEnumerator5
    MaskSpawnGroup=false
    Mesh Sampling Type=::NewEnumerator6
    Skeletal Mesh=$System.SkeletalMesh
   )
   &AddVelocity(Velocity=&RandomRangeVector(Minimum=<-1500 50 50>, Maximum=<-2000 -50 -50>))
   &ShapeLocation(Box Size=<0 20 60>, MaskSpawnGroup=false, Override Seed=false, Shape Primitive=::NewEnumerator2, Surface Only=false, UseOwnerScale=true)
  ]
  particleUpdateScript [
   &ParticleState()
   &InheritSourceMovement(
    Applied Position Scale=&Multiply_VectorByFloat(
     Vector=<1 1 1>
     Float=&FloatFromCurve(
      FloatCurve=DataInterface() {propertyValues {Curve {
         defaultValue=0
         keys [
          {tangentMode=RCTM_User, time=0, value=1}
          <at 0.0574 0.5125>
          <at 0.1165 0.2326>
          <at 0.2353 0.0414>
          <at 0.4 0>
          {tangentMode=RCTM_User, time=1, value=0}
         ]
        }}}
      Scale Curve=0.5
     )
    )
    ApplyForce=false
    ApplyPosition=true
    ApplyPositionFalloff=false
   )
   &ScaleMeshSize(Scale Factor=&VectorFromFloat(Value=&FloatFromCurve(FloatCurve=DataInterface() {propertyValues {Curve {
         defaultValue=1
         keys [
          {arriveTangent=-0.1261, interpMode=RCIM_Cubic, leaveTangent=-0.1261, tangentMode=RCTM_User, time=0, value=1}
          {arriveTangent=-2.7986, interpMode=RCIM_Cubic, leaveTangent=-2.7986, tangentMode=RCTM_User, time=1, value=0}
         ]
        }}}))) {enabled=false}
   &DynamicMaterialParameters(
    Index 0 Param 1=&FloatFromCurve(FloatCurve=DataInterface() {propertyValues {Curve {keys [<at 0 1>,<at 0.2198 0.9711>,<at 0.4641 0.3828>,<at 0.6389 0.1131>,<at 1 0>]}}})
    Index 0 Param 2=&FloatFromCurve(FloatCurve=DataInterface() {propertyValues {Curve {keys [<at 0 1>,<at 0.4998 1>,<at 0.7896 0.8315>,<at 1 0>]}}})
    Index 0 Param 3=&FloatFromCurve(FloatCurve=DataInterface() {propertyValues {Curve {keys [<at 0 100>,<at 0.2639 37.405>,<at 0.5064 10.0618>,<at 1 0>]}}}, Scale Curve=10)
    Index 0 Param 4=0
    Index 1 Param 1=&FloatFromCurve(
     FloatCurve=DataInterface() {propertyValues {Curve {
        defaultValue=1
        keys [
         {arriveTangent=-2.906, interpMode=RCIM_Cubic, leaveTangent=-2.906, tangentMode=RCTM_User, time=0, value=1}
         {arriveTangent=0.0224, interpMode=RCIM_Cubic, leaveTangent=0.0224, tangentMode=RCTM_User, time=1, value=0}
        ]
       }}}
     Scale Curve=0.05
    )
    Index 1 Param 2=&FloatFromCurve(FloatCurve=DataInterface() {propertyValues {Curve {keys [<at 0 1>,<at 1 0>]}}}, Scale Curve=10)
    Module.Index0Float0Write=true
    Module.Index0Float1Write=true
    Module.Index0Float2Write=true
    Module.Index0Float3Write=true
    Module.Index1Float0Write=true
    Module.Index1Float1Write=true
    Module.Index1Float2Write=true
    Module.Index1Float3Write=true
    Write Parameter Index 1=::NewEnumerator1
   )
   &Drag(Drag=10, Use Linear Drag=true, Use Rotational Drag=true) {enabled=false}
   &SolveForcesAndVelocity(Clamp Velocity=false, Limit Acceleration=false)
   &ScaleColor(
    Scale RGB=&MakeVectorFromLinearColorRGB(Color=&ColorFromCurve(ColorCurve=DataInterface() {propertyValues {
        AlphaCurve {keys [<at 0 1 RCIM_Cubic>,<at 1 0 RCIM_Cubic>]}
        BlueCurve {keys [<at 0.202 20>,<at 0.4466 0>,<at 0.6664 0.0002>]}
        GreenCurve {keys [<at 0.202 0.5>,<at 0.4466 1.6296>,<at 0.6664 0.0253>]}
        RedCurve {keys [<at 0.202 1>,<at 0.4466 9.4792>,<at 0.6664 9.4792>]}
       }}))
    ScaleA=true
    ScaleRGB=true
   )
  ]
  rendererClasses [&NiagaraMeshRendererProperties, &NiagaraSpriteRendererProperties]
  renderers [
   {rendererClass=&NiagaraMeshRendererProperties, rendererIndex=0}
   {rendererClass=&NiagaraSpriteRendererProperties, rendererIndex=1}
  ]
 }
 NE_VC_Boost_Sprite_Parent001 {
  emitterUpdateScript [
   &EmitterState()
   &VC_FilteredSocketCount()
   &SpawnRate(SpawnRate=100, Use Spawn Probability=false)
  ]
  particleSpawnScript [
   &VC_PaintedVariantsOverride() {enabled=false}
   &InitializeParticle(
    Lifetime=&RandomRangeFloat(Minimum=0.3, Maximum=0.5)
    Sprite Rotation Mode=::NewEnumerator0
    Sprite Size Max=<85 300>
    Sprite Size Min=<70 50>
    Sprite Size Mode=::NewEnumerator2
    UsePositionOffset=false
    Write Material Random=true
   )
   &PartitionParticles(Number Of Partitions=$Emitter.PortCount)
   &SkeletalMeshLocation(
    Bone / Socket Index=$Particles.PartitionParticles.Partition
    Bone / Socket Sampling Mode=::NewEnumerator5
    MaskSpawnGroup=false
    Mesh Sampling Type=::NewEnumerator6
    Sampled Position Offset=<-10 0 0>
    Skeletal Mesh=$System.SkeletalMesh
   )
   &AddVelocity(Velocity=&RandomRangeVector(Minimum=<-500 0 0>, Maximum=<-1000 1 0>))
   &ShapeLocation(Box Size=<0 0 20>, MaskSpawnGroup=false, Override Seed=false, Shape Primitive=::NewEnumerator2, Surface Only=false, UseOwnerScale=true)
  ]
  particleUpdateScript [
   &ParticleState()
   &InheritSourceMovement(
    Applied Position Scale=&Multiply_VectorByFloat(
     Vector=<1 1 1>
     Float=&FloatFromCurve(FloatCurve=DataInterface() {propertyValues {Curve {keys [<at 0.3 1 RCIM_Cubic>, {arriveTangent=-3.5697, interpMode=RCIM_Cubic, leaveTangent=-3.5697, tangentMode=RCTM_User, time=1, value=0}]}}})
    )
    ApplyForce=false
    ApplyPosition=true
    ApplyPositionFalloff=false
   )
   &ScaleSpriteSize(
    Non-Uniform Curve Sprite Scale=DataInterface() {propertyValues {
      XCurve {keys [{tangentMode=RCTM_User, time=0, value=1}, {tangentMode=RCTM_User, time=1.003, value=0.9902}]}
      YCurve {keys [{tangentMode=RCTM_User, time=0.0089, value=0.2459}, <at 0.0746 0.4897>, <at 0.1901 0.7059>, {tangentMode=RCTM_User, time=1, value=1}]}
     }}
    Scale Sprite Size Mode=::NewEnumerator3
   )
   &Drag(Drag=2, Use Linear Drag=true, Use Rotational Drag=true)
   &SolveForcesAndVelocity(Clamp Velocity=false, Limit Acceleration=false)
   &DynamicMaterialParameters(
    Index 0 Param 1=&FloatFromCurve(FloatCurve=DataInterface() {propertyValues {Curve {keys [<at 0.004 0.0118>,<at 0.0975 0.8>,<at 0.4959 0.2126>,<at 0.6667 0.0593>,<at 1 0>]}}})
    Index 0 Param 2=&FloatFromCurve(FloatCurve=DataInterface() {propertyValues {Curve {keys [<at 0.002 0.9882>,<at 0.2436 0.5336>,<at 0.4621 0.1854>,<at 0.7203 0.021>,<at 1.004 -0.0118>]}}})
    Index 0 Param 3=&FloatFromCurve(FloatCurve=DataInterface() {propertyValues {Curve {keys [<at 0 10>,<at 1 5>]}}}, Scale Curve=0)
    Index 0 Param 4=&FloatFromCurve(FloatCurve=DataInterface() {propertyValues {Curve {keys [<at 0 0>,<at 1 1>]}}}, Scale Curve=0)
    Index 1 Param 1=&FloatFromCurve(
     FloatCurve=DataInterface() {propertyValues {Curve {
        defaultValue=1
        keys [
         {arriveTangent=-3.8446, interpMode=RCIM_Cubic, leaveTangent=-3.8446, tangentMode=RCTM_User, time=0, value=1}
         {arriveTangent=0.0224, interpMode=RCIM_Cubic, leaveTangent=0.0224, tangentMode=RCTM_User, time=1, value=0}
        ]
       }}}
     Scale Curve=0.02
    )
    Index 1 Param 2=&FloatFromCurve(FloatCurve=DataInterface() {propertyValues {Curve {keys [<at 0 1>,<at 1.0058 0.5>]}}}, Scale Curve=20)
    Module.Index0Float0Write=true
    Module.Index0Float1Write=true
    Module.Index0Float2Write=true
    Module.Index0Float3Write=true
    Module.Index1Float0Write=true
    Module.Index1Float1Write=true
    Module.Index1Float2Write=true
    Module.Index1Float3Write=true
    Write Parameter Index 1=::NewEnumerator1
   )
   &ScaleColor(
    Scale Alpha=&FloatFromCurve(FloatCurve=DataInterface() {propertyValues {Curve {keys [<at 0 0>,<at 0.1024 0.8>,<at 1 0>]}}})
    Scale RGB=&MakeVectorFromLinearColorRGB(Color=&ColorFromCurve(ColorCurve=DataInterface() {propertyValues {
        AlphaCurve {keys [<at 0 1 RCIM_Cubic>,<at 0.6485 1>,<at 1 0 RCIM_Cubic>]}
        BlueCurve {keys [<at 0.0085 50>,<at 0.1788 5>,<at 0.2655 4.1976>,<at 0.3599 0.2117>]}
        GreenCurve {keys [<at 0.0085 0.621>,<at 0.1788 1.267>,<at 0.2655 0.797>,<at 0.3599 1.1396>]}
        RedCurve {keys [<at 0.0085 1>,<at 0.1788 0.4281>,<at 0.2655 5>,<at 0.3599 5>]}
       }}))
    ScaleA=true
    ScaleRGB=true
   )
   &Light_Attributes(
    Diffuse Scale=1
    Falloff Exponent=1
    Light Diffuse Scale:NiagaraFloat=&FloatFromCurve(
     FloatCurve=DataInterface() {propertyValues {Curve {defaultValue=1, keys [{arriveTangent=-1.9574, interpMode=RCIM_Cubic, leaveTangent=-1.9574, tangentMode=RCTM_User, time=0, value=1}]}}}
     Scale Curve=0.3
    )
    Light Exponent=8
    Light Radius=300
    Light Specular Scale:NiagaraFloat=&FloatFromCurve(
     FloatCurve=DataInterface() {propertyValues {Curve {defaultValue=1, keys [{arriveTangent=-2.6849, interpMode=RCIM_Cubic, leaveTangent=-2.6849, tangentMode=RCTM_User, time=0, value=1}]}}}
     Scale Curve=0.3
    )
    Radius=1
    Specular Scale=1
   )
  ]
  rendererClasses [&NiagaraLightRendererProperties, &NiagaraSpriteRendererProperties]
  renderers [
   {rendererClass=&NiagaraLightRendererProperties, rendererIndex=0}
   {rendererClass=&NiagaraSpriteRendererProperties, rendererIndex=1}
  ]
 }
 NE_VC_Boost_Sprite_Parent002 {
  emitterUpdateScript [
   &EmitterState()
   &VC_FilteredSocketCount()
   &SpawnRate(SpawnRate=65, Use Spawn Probability=false)
  ]
  particleSpawnScript [
   &VC_PaintedVariantsOverride() {enabled=false}
   &InitializeParticle(
    Lifetime=&RandomRangeFloat(Minimum=1, Maximum=1)
    Sprite Rotation Mode=::NewEnumerator2
    Sprite Size Mode=::NewEnumerator1
    Uniform Sprite Size Max=220
    Uniform Sprite Size Min=200
    UsePositionOffset=false
    Write Material Random=true
   )
   &PartitionParticles(Number Of Partitions=$Emitter.PortCount)
   &SkeletalMeshLocation(
    Bone / Socket Index=$Particles.PartitionParticles.Partition
    Bone / Socket Sampling Mode=::NewEnumerator5
    MaskSpawnGroup=false
    Mesh Sampling Type=::NewEnumerator6
    Skeletal Mesh=$System.SkeletalMesh
   )
   &ShapeLocation(MaskSpawnGroup=false, Override Seed=false, Sphere Radius=10, UseOwnerScale=true)
   &AddVelocity(Velocity=&RandomRangeVector(Minimum=<-800 0 0>, Maximum=<-1000 0 0>))
  ]
  particleUpdateScript [
   &ParticleState()
   &InheritSourceMovement(
    Applied Position Scale=&Multiply_VectorByFloat(
     Vector=<1 1 1>
     Float=&FloatFromCurve(
      FloatCurve=DataInterface() {propertyValues {Curve {
         defaultValue=0
         keys [
          {tangentMode=RCTM_User, time=0, value=1}
          <at 0.03 0.489>
          <at 0.06 0.1856>
          <at 0.1 0.0532>
          <at 0.15 0>
          {tangentMode=RCTM_User, time=1, value=0}
         ]
        }}}
      Scale Curve=0.5
     )
    )
    ApplyForce=false
    ApplyPosition=true
    ApplyPositionFalloff=false
   )
   &ScaleSpriteSize(Uniform Curve Sprite Scale=DataInterface() {propertyValues {Curve {
       defaultValue=1
       keys [
        {arriveTangent=0.9045, interpMode=RCIM_Cubic, leaveTangent=0.9045, tangentMode=RCTM_User, time=0, value=0.4}
        {arriveTangent=0.313, interpMode=RCIM_Cubic, leaveTangent=0.313, tangentMode=RCTM_User, time=1, value=2.6}
       ]
      }}})
   &GravityForce(Gravity=&RandomRangeVector(Minimum=<0 0 0>, Maximum=<1 1 200>)) {enabled=false}
   &SubUVAnimation(
    "Sprite Renderer "=DataInterface() {propertyValues {SpriteRenderer=&self:NE_VC_Boost_Sprite_Parent002.NiagaraSpriteRendererProperties_0}}
    UseEndFrame=false
    UseStartFrame=false
   ) {module=SubUVAnimation001}
   &Drag(Drag=2, Use Linear Drag=true, Use Rotational Drag=true) {enabled=false}
   &SolveForcesAndVelocity(Clamp Velocity=false, Limit Acceleration=false)
   &DynamicMaterialParameters(
    Index 0 Param 1=&FloatFromCurve(FloatCurve=DataInterface() {propertyValues {Curve {
        defaultValue=1
        keys [
         {arriveTangent=-0.1129, interpMode=RCIM_Cubic, leaveTangent=-0.1129, tangentMode=RCTM_User, time=0, value=1}
         {arriveTangent=0.2155, interpMode=RCIM_Cubic, leaveTangent=0.2155, tangentMode=RCTM_User, time=0.1807, value=1}
         {arriveTangent=-2.1755, interpMode=RCIM_Cubic, leaveTangent=-2.1755, tangentMode=RCTM_User, time=1, value=0}
        ]
       }}})
    Index 0 Param 2=&FloatFromCurve(FloatCurve=DataInterface() {propertyValues {Curve {
        defaultValue=1
        keys [
         {arriveTangent=0.1261, interpMode=RCIM_Cubic, leaveTangent=0.1261, tangentMode=RCTM_User, time=0, value=1}
         {arriveTangent=0.0224, interpMode=RCIM_Cubic, leaveTangent=0.0224, tangentMode=RCTM_User, time=1, value=0}
        ]
       }}})
    Index 0 Param 3=1
    Index 0 Param 4=&FloatFromCurve(
     FloatCurve=DataInterface() {propertyValues {Curve {
        defaultValue=1
        keys [
         {arriveTangent=-6.305, interpMode=RCIM_Cubic, leaveTangent=-6.305, tangentMode=RCTM_User, time=0, value=1}
         {arriveTangent=-0.1121, interpMode=RCIM_Cubic, leaveTangent=-0.1121, tangentMode=RCTM_User, time=0.0515, value=0.1356}
         {arriveTangent=0.0224, interpMode=RCIM_Cubic, leaveTangent=0.0224, tangentMode=RCTM_User, time=1, value=0}
        ]
       }}}
     Scale Curve=5
    )
    Module.Index0Float0Write=true
    Module.Index0Float1Write=true
    Module.Index0Float2Write=true
    Module.Index0Float3Write=true
   )
   &ScaleColor(
    Scale Alpha=&FloatFromCurve(
     FloatCurve=DataInterface() {propertyValues {Curve {
        defaultValue=1
        keys [
         {arriveTangent=7.4088, interpMode=RCIM_Cubic, leaveTangent=7.4088, tangentMode=RCTM_User, time=0, value=0}
         {arriveTangent=0.4775, interpMode=RCIM_Cubic, leaveTangent=0.4775, tangentMode=RCTM_User, time=0.15, value=0.8}
         {arriveTangent=-1.8018, interpMode=RCIM_Cubic, leaveTangent=-1.8018, tangentMode=RCTM_User, time=1, value=0}
        ]
       }}}
     Scale Curve=0.8
    )
    Scale RGB=&MakeVectorFromLinearColorRGB(Color=&ColorFromCurve(ColorCurve=DataInterface() {propertyValues {
        AlphaCurve {keys [<at 0 1 RCIM_Cubic>,<at 1 0 RCIM_Cubic>]}
        BlueCurve {keys [<at 0 2>,<at 0.1087 0.2551>,<at 0.2297 0.401>]}
        GreenCurve {keys [<at 0 0.7586>,<at 0.1087 0.4006>,<at 0.2297 0.365>]}
        RedCurve {keys [<at 0 0.5161>,<at 0.1087 0.7083>,<at 0.2297 0.3384>]}
       }}))
    ScaleA=true
    ScaleRGB=true
   )
   &CameraOffset(Camera Offset Amount=5)
  ]
  rendererClasses [&NiagaraSpriteRendererProperties]
  renderers [{rendererClass=&NiagaraSpriteRendererProperties}]
 }
 NE_VC_Boost_Sprite_Parent003 {
  emitterUpdateScript [
   &EmitterState()
   &VC_FilteredSocketCount()
   &SpawnRate(SpawnRate=500, Use Spawn Probability=false)
  ]
  particleSpawnScript [
   &VC_PaintedVariantsOverride() {enabled=false}
   &InitializeParticle(
    Lifetime=&RandomRangeFloat(Minimum=1, Maximum=1)
    Sprite Rotation Mode=::NewEnumerator2
    Sprite Size Mode=::NewEnumerator1
    Uniform Sprite Size Max=250
    Uniform Sprite Size Min=230
    UsePositionOffset=false
    Write Material Random=true
   )
   &PartitionParticles(Number Of Partitions=$Emitter.PortCount)
   &SkeletalMeshLocation(
    Bone / Socket Index=$Particles.PartitionParticles.Partition
    Bone / Socket Sampling Mode=::NewEnumerator5
    MaskSpawnGroup=false
    Mesh Sampling Type=::NewEnumerator6
    Skeletal Mesh=$System.SkeletalMesh
   )
   &ShapeLocation(MaskSpawnGroup=false, Override Seed=false, Sphere Radius=10, UseOwnerScale=true)
   &AddVelocity(Velocity=&RandomRangeVector(Minimum=<-800 0 0>, Maximum=<-1000 0 0>))
  ]
  particleUpdateScript [
   &ParticleState()
   &InheritSourceMovement(
    Applied Position Scale=&Multiply_VectorByFloat(
     Vector=<1 1 1>
     Float=&FloatFromCurve(
      FloatCurve=DataInterface() {propertyValues {Curve {
         defaultValue=0
         keys [
          {tangentMode=RCTM_User, time=0, value=1}
          <at 0.03 0.489>
          <at 0.06 0.1856>
          <at 0.1 0.0532>
          <at 0.15 0>
          {tangentMode=RCTM_User, time=1, value=0}
         ]
        }}}
      Scale Curve=0.5
     )
    )
    ApplyForce=false
    ApplyPosition=true
    ApplyPositionFalloff=false
   )
   &ScaleSpriteSize(Uniform Curve Sprite Scale=DataInterface() {propertyValues {Curve {
       defaultValue=1
       keys [
        {arriveTangent=0.0817, interpMode=RCIM_Cubic, leaveTangent=0.0817, tangentMode=RCTM_User, time=0, value=0.15}
        {arriveTangent=4.8234, interpMode=RCIM_Cubic, leaveTangent=4.8234, tangentMode=RCTM_User, time=0.0971, value=0.6644}
        {arriveTangent=2.3264, interpMode=RCIM_Cubic, leaveTangent=2.3264, tangentMode=RCTM_User, time=0.5295, value=2.3476}
        <at 0.9977 3 RCIM_Cubic>
       ]
      }}})
   &GravityForce(Gravity=&RandomRangeVector(Minimum=<0 0 100>, Maximum=<1 1 -100>))
   &SubUVAnimation(
    "Sprite Renderer "=DataInterface() {propertyValues {SpriteRenderer=&self:NE_VC_Boost_Sprite_Parent003_0.NiagaraSpriteRendererProperties_0}}
    UseEndFrame=false
    UseStartFrame=false
   ) {module=SubUVAnimation001}
   &Drag(Drag=4, Use Linear Drag=true, Use Rotational Drag=true) {enabled=false}
   &SolveForcesAndVelocity(Clamp Velocity=false, Limit Acceleration=false)
   &DynamicMaterialParameters(
    Index 0 Param 1=&FloatFromCurve(
     FloatCurve=DataInterface() {propertyValues {Curve {
        defaultValue=1
        keys [
         {arriveTangent=13.0726, interpMode=RCIM_Cubic, leaveTangent=13.0726, tangentMode=RCTM_User, time=0, value=0}
         {arriveTangent=-2.0737, interpMode=RCIM_Cubic, leaveTangent=-2.0737, tangentMode=RCTM_User, time=0.05, value=0.55}
         {arriveTangent=-0.5789, interpMode=RCIM_Cubic, leaveTangent=-0.5789, time=0.1, value=0.25}
         {arriveTangent=-0.1781, interpMode=RCIM_Cubic, leaveTangent=-0.1781, tangentMode=RCTM_User, time=1, value=0}
        ]
       }}}
     Scale Curve=0.5
    )
    Index 0 Param 2=&FloatFromCurve(FloatCurve=DataInterface() {propertyValues {Curve {
        defaultValue=1
        keys [
         {arriveTangent=0.1261, interpMode=RCIM_Cubic, leaveTangent=0.1261, tangentMode=RCTM_User, time=0, value=1}
         {arriveTangent=0.0224, interpMode=RCIM_Cubic, leaveTangent=0.0224, tangentMode=RCTM_User, time=1, value=0}
        ]
       }}})
    Index 0 Param 3=1
    Index 0 Param 4=&FloatFromCurve(
     FloatCurve=DataInterface() {propertyValues {Curve {
        defaultValue=1
        keys [
         {arriveTangent=-1.9611, interpMode=RCIM_Cubic, leaveTangent=-1.9611, tangentMode=RCTM_User, time=0, value=1}
         {arriveTangent=0.0224, interpMode=RCIM_Cubic, leaveTangent=0.0224, tangentMode=RCTM_User, time=0.5, value=0.6}
        ]
       }}}
     Scale Curve=0.7
    )
    Module.Index0Float0Write=true
    Module.Index0Float1Write=true
    Module.Index0Float2Write=true
    Module.Index0Float3Write=true
    Module.Index1Float0Write=true
    Module.Index1Float1Write=true
    Module.Index1Float2Write=true
    Module.Index1Float3Write=true
    Module.Index2Float0Write=true
    Module.Index2Float1Write=true
    Module.Index2Float2Write=true
    Module.Index2Float3Write=true
    Module.Index3Float0Write=true
    Module.Index3Float1Write=true
    Module.Index3Float2Write=true
    Module.Index3Float3Write=true
    Write Parameter Index 1=::NewEnumerator1
    Write Parameter Index 2=::NewEnumerator1
    Write Parameter Index 3=::NewEnumerator1
   )
   &ScaleColor(
    Scale Alpha=&FloatFromCurve(
     FloatCurve=DataInterface() {propertyValues {Curve {
        defaultValue=1
        keys [
         {arriveTangent=26.5855, interpMode=RCIM_Cubic, leaveTangent=26.5855, tangentMode=RCTM_User, time=0, value=0}
         {arriveTangent=20, interpMode=RCIM_Cubic, leaveTangent=20, time=0.0125, value=0.588}
         {arriveTangent=0.0361, interpMode=RCIM_Cubic, leaveTangent=0.0361, tangentMode=RCTM_User, time=0.05, value=1}
         {arriveTangent=-0.9095, interpMode=RCIM_Cubic, leaveTangent=-0.9095, tangentMode=RCTM_User, time=1, value=0}
        ]
       }}}
     Scale Curve=1.3
    )
    Scale RGB=&MakeVectorFromLinearColorRGB(Color=&ColorFromCurve(ColorCurve=DataInterface() {propertyValues {
        AlphaCurve {keys [<at 0 1 RCIM_Cubic>,<at 1 0 RCIM_Cubic>]}
        BlueCurve {keys [<at 0.0244 2>,<at 0.059 0.4217>,<at 0.0825 0.5383>,<at 0.1643 1>]}
        GreenCurve {keys [<at 0.0244 0.7048>,<at 0.059 0.7666>,<at 0.0825 0.5765>,<at 0.1643 0.7663>]}
        RedCurve {keys [<at 0.0244 0.174>,<at 0.059 3>,<at 0.0825 1>,<at 0.1643 0.5938>]}
       }}))
    ScaleA=true
    ScaleRGB=true
   )
   &CameraOffset(Camera Offset Amount=5)
  ]
  rendererClasses [&NiagaraSpriteRendererProperties]
  renderers [{rendererClass=&NiagaraSpriteRendererProperties}]
 }
}
summary {
 systemName=NS_Boost_Plexus_bckp
 emitters [
  {emitterName=NE_VC_Boost_Ribbon_Parent, rendererClasses [&NiagaraRibbonRendererProperties]}
  {emitterName=NE_VC_Boost_Ribbon_Parent002, rendererClasses [&NiagaraRibbonRendererProperties]}
  {
   emitterName=NE_VC_Boost_Sprite_Parent
   rendererClasses [&NiagaraMeshRendererProperties, &NiagaraSpriteRendererProperties]
  }
  {
   emitterName=NE_VC_Boost_Sprite_Parent001
   rendererClasses [&NiagaraLightRendererProperties, &NiagaraSpriteRendererProperties]
  }
  {emitterName=NE_VC_Boost_Sprite_Parent002, rendererClasses [&NiagaraSpriteRendererProperties]}
  {emitterName=NE_VC_Boost_Sprite_Parent003, rendererClasses [&NiagaraSpriteRendererProperties]}
 ]
 userVariables [
  User.bIsFrontendPreview {type=NiagaraBool}
  User.bIsPainted {type=NiagaraBool}
  User.ColorVariantID {type=NiagaraInt32}
  User.PaintedColor {type=LinearColor, description="", defaultValue {struct=LinearColor, value=<rgba 0.0018 0.0938 0.085 1>}}
  User.PaintedScalar {type=NiagaraFloat, defaultValue=1, description=""}
  User.PortCount {type=NiagaraInt32, defaultValue=2}
 ]
}
